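; Actor Blueprint that acts as a controller for persistent event audio, tracking spawned sounds by Name in a Name-to-AudioComponent map.
; Play functions spawn located or attached sounds with fade-in, ignoring Names already in the map.
; Other functions fade out and remove one sound or set float parameters on playing ones.
; StopAllEventSounds, called on authority BeginPlay and on EndPlay, fades out and clears a non-empty map, optionally destroying the actor.
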(event EventBeginPlay
  (Networking|SwitchHasAuthority
    (:Authority
      (CallFunction|StopAllEventSounds true))))

(event EventEndPlay
  (params (EndPlayReason :type EEndPlayReason%20Enum :default "Destroyed"))
  (CallFunction|StopAllEventSounds true))

(fn PlayEventSoundAtLocation
  (params (Sound :type Sound%20Base%20Object%20Reference) (Location :type Vector) (FadeInDuration :type Float%20%28single-precision%29) (NameID :type Name))
  (if (not (Utilities|Map|Contains (Variables|Default|GetNameToAudioComponent) NameID))
    (bind _returnvalue (Audio|SpawnSoundatLocation Sound Location))
    (Audio|Components|Audio|FadeIn _returnvalue FadeInDuration 1.0 0.0 "Linear" :!class /Script/Engine.AudioComponent)
    (Utilities|Map|Add (Variables|Default|GetNameToAudioComponent) NameID _returnvalue)))

(fn PlayEventSoundAttached
  (params (Sound :type Sound%20Base%20Object%20Reference) (AttachToComponent :type Scene%20Component%20Object%20Reference) (AttachPointName :type Name) (FadeInDuration :type Float%20%28single-precision%29) (NameID :type Name%20%28by%20ref%29))
  (if (not (Utilities|Map|Contains (Variables|Default|GetNameToAudioComponent) NameID))
    (bind _returnvalue (Audio|SpawnSoundAttached Sound AttachToComponent AttachPointName))
    (Audio|Components|Audio|FadeIn _returnvalue FadeInDuration 1.0 0.0 "Linear" :!class /Script/Engine.AudioComponent)
    (Utilities|Map|Add (Variables|Default|GetNameToAudioComponent) NameID _returnvalue)))

(fn StopAllEventSounds
  (params (bDestroyActor :type Boolean :default true))
  (bind _nametoaudiocomponent (Variables|Default|GetNameToAudioComponent))
  (if (> (Utilities|Map|Length _nametoaudiocomponent) 0)
    (bind _values (Utilities|Map|Values _nametoaudiocomponent))
    (for _array_element _values
      (Utilities|IsValid _array_element
        (:Is%20Valid
          (Audio|Components|Audio|FadeOut _array_element 0.5 0.0 "Linear" :!class /Script/Engine.AudioComponent))))
    (Utilities|Map|Clear _nametoaudiocomponent)
    (if bDestroyActor
      (Actor|DestroyActor))))

(fn StopEventSound
  (params (FadeOutDuration :type Float%20%28single-precision%29) (NameID :type Name))
  (bind _nametoaudiocomponent (Variables|Default|GetNameToAudioComponent))
  (bind (_value _returnvalue) (Utilities|Map|Find _nametoaudiocomponent NameID))
  (if _returnvalue
    (Utilities|IsValid _value
      (:Is%20Valid
        (Audio|Components|Audio|FadeOut _value FadeOutDuration 0.0 "Linear" :!class /Script/Engine.AudioComponent)
        (Utilities|Map|Remove _nametoaudiocomponent NameID)))))

(fn UpdateEventSoundParameter
  (params (NameID :type Name :default "Sound") (ParamName :type Name) (ParamFloatValue :type Float%20%28single-precision%29))
  (bind _nametoaudiocomponent (Variables|Default|GetNameToAudioComponent))
  (bind (_value _returnvalue) (Utilities|Map|Find _nametoaudiocomponent NameID))
  (if _returnvalue
    (Utilities|IsValid _value
      (:Is%20Valid
        (if (Audio|Components|Audio|IsPlaying _value)
          (Audio|Parameter|SetFloatParameter _value ParamName ParamFloatValue :!class /Script/AudioExtensions.AudioParameterControllerInterface)
          (else
            _)))
      (:Is%20Not%20Valid
        _))
    (else
      _))
  (return))

(fn ConstructionScript)
